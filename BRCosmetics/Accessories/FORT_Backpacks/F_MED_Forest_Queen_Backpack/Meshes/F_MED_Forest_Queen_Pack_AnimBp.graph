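# Animation Blueprint for the Forest Queen backpack, adding secondary motion to its cape and leaf bones.
# Copies the parent mesh's pose, then shapes the cape with trail controllers, angular limits and a backward-pose blend, bypassed when cloth is disabled.
# Weighted constraints attach leaves to root, cape and leaf bones before per-leaf Anim Dynamics adds damped sway.
# The event graph derives movement values, including the trail alpha, from player movement.
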AnimBlueprint /BRCosmetics/Accessories/FORT_Backpacks/F_MED_Forest_Queen_Backpack/Meshes/F_MED_Forest_Queen_Pack_AnimBp
target_skeleton=/Game/Accessories/FORT_Backpacks/F_MED_Forest_Queen_Backpack/Meshes/F_MED_Forest_Queen_Pack_Skeleton
graphs [
 AnimGraph {nodes [
   n1 = Trail controller - Bone: dyn_cape05!(
    ComponentPose=Local To Component(LocalPose=Use cached pose 'Cape_Cache'())
    Alpha=Get BwdMovement() -> (BwdMovement:double)
    FakeVelocity:Vector="0,0,0"
   ) -> (Pose) {
    bActorSpaceFakeVel=true
    bLimitRotation=true
    bUsePlanarLimit=true
    chainLength=5
    lastBoneRotationAnimAlphaBlend=0.5
    maxDeltaTime=0.0333
    baseJoint {boneName=backpack_root}
    fakeVelocity {y=50, z=-50}
    planarLimits [
     {drivingBone {boneName=backpack_root}, planeTransform {location=<0 -2.0567 0>, rotation=<rot 0 0 -90>, scale=<1 1 1>}}
     {drivingBone {boneName=backpack_root}, planeTransform {location=<0 0 10>, rotation=<rot 0 0 -150>, scale=<1 1 1>}}
    ]
    rotationLimits [{limitMax=<180 180 15>}, {limitMax=<180 180 25>}, {limitMax=<180 180 35>}, {limitMax=<180 180 35>}, {limitMax=<180 180 35>}]
    rotationOffsets [<0 0 0>, <0 0 0>, <0 0 0>, <0 0 0>, <0 0 0>]
    trailBone {boneName=dyn_cape05}
    trailRelaxationSpeed {editorCurveData {keys [<at 0 20 RCIM_Linear>, {arriveTangent=-14.4971, interpMode=RCIM_Linear, leaveTangent=-14.4971, time=1, value=10}]}}
   }
   n2 = Apply Limits!(ComponentPose:Component Space Pose Link Structure=@n1, Alpha:float=1, AngularOffsets_0:Vector=Get AimOffset() -> (AimOffset:Vector)) -> (Pose:Component Space Pose Link Structure) {
    is=title AnimGraphNode_ApplyLimits
    alphaBoolBlend {blendOption=Linear}
    angularOffsets [<0 0 0>]
    angularRangeLimits [{bone {boneName=dyn_cape01}, limitMax=<10 180 0>, limitMin=<-10 -180 -120>}]
   }
   n3 = Component To Local(ComponentPose=@n2)
   n4 = F_MED_Forest_Queen_Backpack_Pose_PoseAsset : BackwardPose!(PoseAsset:Pose Asset Object Reference=&/F_MED_Forest_Queen_Backpack_Pose_PoseAsset) -> (Pose:Pose Link Structure) {is=title AnimGraphNode_PoseByName, poseName=BackwardPose}
   n5 = Get bBlendInBackwardPose() -> (bBlendInBackwardPose:bool)
   n6 = "Transform (Modify) Bone - Bone: dyn_cape01"(
    ComponentPose=Local To Component(LocalPose=Two Way Blend!(A=@n3, B=@n4, bAlphaBoolEnabled:bool=@n5) -> (Pose) {blendNode {alphaInputType=Bool, alphaBoolBlend {blendInTime=0.5, blendOption=Linear, blendOutTime=0.5}}})
    Alpha=1
    Translation="0,0,0"
    Rotation="0,0,15"
    Scale="1,1,1"
   ) {
    rotationMode=BMM_Additive
    translationMode=BMM_Additive
    alphaBoolBlend {blendOption=Linear}
    alphaScaleBiasClamp {bClampResult=true, bInterpResult=true}
    boneToModify {boneName=dyn_cape01}
    rotation {roll=15}
   }
   n7 = "Transform (Modify) Bone - Bone: dyn_cape01"(ComponentPose=@n6, Alpha=1, Translation="0,0,0", Rotation="0,0,35", Scale="1,1,1") {
    rotationMode=BMM_Additive
    translationMode=BMM_Additive
    alphaBoolBlend {blendOption=Linear}
    boneToModify {boneName=dyn_cape01}
    rotation {roll=35}
   }
   n8 = Get BwdMovement()
   n9 = Get FinalCapeWindVelocity() -> (FinalCapeWindVelocity:Vector)
   n10 = Trail controller - Bone: dyn_cape06_l!(ComponentPose=@n7, Alpha=@n8, FakeVelocity:Vector=@n9) -> (Pose) {
    bActorSpaceFakeVel=true
    bLimitRotation=true
    chainLength=6
    lastBoneRotationAnimAlphaBlend=1
    maxDeltaTime=0.0333
    alphaScaleBiasClamp {bInterpResult=true}
    baseJoint {boneName=backpack_root}
    planarLimits [{drivingBone {boneName=dyn_cape05}, planeTransform {location=<0 -2.0567 0>, rotation=<rot 0 0 -90>, scale=<1 1 1>}}]
    rotationLimits [
     {limitMax=<180 180 0>, limitMin=<-180 -180 0>}
     {limitMax=<180 180 0>, limitMin=<-180 -180 0>}
     {limitMax=<180 180 0>, limitMin=<-180 -180 0>}
     {limitMax=<180 180 0>, limitMin=<-180 -180 0>}
     {limitMax=<180 180 0>, limitMin=<-180 -180 0>}
     {limitMax=<180 180 25>}
    ]
    rotationOffsets [<0 0 0>, <0 0 0>, <0 0 0>, <0 0 0>, <0 0 0>, <0 0 0>]
    trailBone {boneName=dyn_cape06_l}
    trailRelaxationSpeed {editorCurveData {keys [<at 0 15 RCIM_Linear>,<at 1 6 RCIM_Linear>]}}
   }
   n11 = Trail controller - Bone: dyn_cape06_r!(ComponentPose=@n10, Alpha=@n8, FakeVelocity:Vector=@n9) -> (Pose) {
    bActorSpaceFakeVel=true
    bLimitRotation=true
    chainLength=6
    lastBoneRotationAnimAlphaBlend=1
    maxDeltaTime=0.0333
    alphaScaleBiasClamp {bInterpResult=true}
    baseJoint {boneName=backpack_root}
    planarLimits [{drivingBone {boneName=dyn_cape05}, planeTransform {location=<0 -2.0567 0>, rotation=<rot 0 0 -90>, scale=<1 1 1>}}]
    rotationLimits [
     {limitMax=<180 180 0>, limitMin=<-180 -180 0>}
     {limitMax=<180 180 0>, limitMin=<-180 -180 0>}
     {limitMax=<180 180 0>, limitMin=<-180 -180 0>}
     {limitMax=<180 180 0>, limitMin=<-180 -180 0>}
     {limitMax=<180 180 0>, limitMin=<-180 -180 0>}
     {limitMax=<180 180 25>}
    ]
    rotationOffsets [<0 0 0>, <0 0 0>, <0 0 0>, <0 0 0>, <0 0 0>, <0 0 0>]
    trailBone {boneName=dyn_cape06_r}
    trailRelaxationSpeed {editorCurveData {keys [<at 0 15 RCIM_Linear>,<at 1 8 RCIM_Linear>]}}
   }
   n12 = Get bCanPlayCustomAnimations() -> (bCanPlayCustomAnimations:bool)
   n13 = "Transform (Modify) Bone - Bone: dyn_cape01"!(ComponentPose=@n11, Alpha="To Float (Boolean)"(InBool=NOT Boolean(A=@n12)), Rotation="0,0,-10") -> (Pose) {rotationMode=BMM_Additive, alphaBoolBlend {blendOption=Linear}, boneToModify {boneName=dyn_cape01}, rotation {roll=-10}}
   n14 = Component To Local(ComponentPose=@n13)
   n15 = Use cached pose 'Cape_Cache'()
   n16 = Get bClothEnabled()
   n17 = "Constraint - dyn_leaf_01 (PP)"!(
    ComponentPose=Local To Component(LocalPose=Blend Poses by bool(BlendPose_0=@n15, BlendPose_1=@n14, BlendTime_0=0, BlendTime_1=0, bActiveValue=@n16) {blendType=Linear})
    Alpha=1
    ConstraintWeights_0=0.9013
    ConstraintWeights_1:float=0.0987
   ) -> (Pose) {
    alphaBoolBlend {blendOption=Linear}
    boneToModify {boneName=dyn_leaf_01}
    constraintSetup [{targetBone {boneName=backpack_root}}, {targetBone {boneName=dyn_leaf_01}}]
    constraintWeights [1,1]
   }
   n18 = "Constraint - dyn_leaf_02 (PPP)"!(ComponentPose=@n17, Alpha=1, ConstraintWeights_0=0.9997, ConstraintWeights_1:float=0.0002, ConstraintWeights_2:float=0.0002) -> (Pose) {
    alphaBoolBlend {blendOption=Linear}
    boneToModify {boneName=dyn_leaf_02}
    constraintSetup [{targetBone {boneName=backpack_root}}, {targetBone {boneName=dyn_cape02}}, {targetBone {boneName=dyn_cape01}}]
    constraintWeights [1,1,1]
   }
   n19 = "Constraint - dyn_leaf_03 (PPP)"!(ComponentPose=@n18, Alpha=1, ConstraintWeights_0=0.9969, ConstraintWeights_1:float=0.0016, ConstraintWeights_2:float=0.0015) -> (Pose) {
    alphaBoolBlend {blendOption=Linear}
    boneToModify {boneName=dyn_leaf_03}
    constraintSetup [{targetBone {boneName=backpack_root}}, {targetBone {boneName=dyn_cape02}}, {targetBone {boneName=dyn_cape01}}]
    constraintWeights [1,1,1]
   }
   n20 = "Constraint - dyn_leaf_04 (PP)"!(ComponentPose=@n19, Alpha=1, ConstraintWeights_0=0.4643, ConstraintWeights_1:float=0.5357) -> (Pose) {
    alphaBoolBlend {blendOption=Linear}
    boneToModify {boneName=dyn_leaf_04}
    constraintSetup [{targetBone {boneName=backpack_root}}, {targetBone {boneName=dyn_leaf_04}}]
    constraintWeights [1,1]
   }
   n21 = "Constraint - dyn_leaf_05 (PPPP)"!(
    ComponentPose=@n20
    Alpha=1
    ConstraintWeights_0=0.995
    ConstraintWeights_1:float=0.0026
    ConstraintWeights_2:float=0.0024
    ConstraintWeights_3:float=0
   ) -> (Pose) {
    alphaBoolBlend {blendOption=Linear}
    boneToModify {boneName=dyn_leaf_05}
    constraintSetup [
     {targetBone {boneName=backpack_root}}
     {targetBone {boneName=dyn_cape02}}
     {targetBone {boneName=dyn_cape01}}
     {targetBone {boneName=dyn_leaf_09}}
    ]
    constraintWeights [1,1,1,1]
   }
   n22 = "Constraint - dyn_leaf_06 (PPP)"!(ComponentPose=@n21, Alpha=1, ConstraintWeights_0=0.5195, ConstraintWeights_1:float=0.2478, ConstraintWeights_2:float=0.2327) -> (Pose) {
    alphaBoolBlend {blendOption=Linear}
    boneToModify {boneName=dyn_leaf_06}
    constraintSetup [{targetBone {boneName=backpack_root}}, {targetBone {boneName=dyn_cape02}}, {targetBone {boneName=dyn_cape01}}]
    constraintWeights [1,1,1]
   }
   n23 = "Constraint - dyn_leaf_07 (PPPPP)"!(
    ComponentPose=@n22
    Alpha=1
    ConstraintWeights_0=0.1025
    ConstraintWeights_1:float=0.4012
    ConstraintWeights_2:float=0.241
    ConstraintWeights_3:float=0.2515
    ConstraintWeights_4:float=0.0037
   ) -> (Pose) {
    alphaBoolBlend {blendOption=Linear}
    boneToModify {boneName=dyn_leaf_07}
    constraintSetup [
     {targetBone {boneName=dyn_cape03}}
     {targetBone {boneName=dyn_cape02}}
     {targetBone {boneName=dyn_cape01}}
     {targetBone {boneName=backpack_root}}
     {targetBone {boneName=dyn_cape04}}
    ]
    constraintWeights [1,1,1,1,1]
   }
   n24 = "Constraint - dyn_leaf_08 (PPPPP)"!(
    ComponentPose=@n23
    Alpha=1
    ConstraintWeights_0=0.7744
    ConstraintWeights_1:float=0.028
    ConstraintWeights_2:float=0.0257
    ConstraintWeights_3:float=0.1718
    ConstraintWeights_4:float=0.0001
   ) -> (Pose) {
    alphaBoolBlend {blendOption=Linear}
    boneToModify {boneName=dyn_leaf_08}
    constraintSetup [
     {targetBone {boneName=backpack_root}}
     {targetBone {boneName=dyn_cape02}}
     {targetBone {boneName=dyn_cape01}}
     {targetBone {boneName=dyn_leaf_08}}
     {targetBone {boneName=dyn_leaf_09}}
    ]
    constraintWeights [1,1,1,1,1]
   }
   n25 = "Constraint - dyn_leaf_09 (PPPP)"!(
    ComponentPose=@n24
    Alpha=1
    ConstraintWeights_0=0.8554
    ConstraintWeights_1:float=0.0034
    ConstraintWeights_2:float=0.0031
    ConstraintWeights_3:float=0.1381
   ) -> (Pose) {
    alphaBoolBlend {blendOption=Linear}
    boneToModify {boneName=dyn_leaf_09}
    constraintSetup [
     {targetBone {boneName=backpack_root}}
     {targetBone {boneName=dyn_cape02}}
     {targetBone {boneName=dyn_cape01}}
     {targetBone {boneName=dyn_leaf_09}}
    ]
    constraintWeights [1,1,1,1]
   }
   n26 = "Constraint - dyn_leaf_10 (PPPPP)"!(
    ComponentPose=@n25
    Alpha=1
    ConstraintWeights_0=0.4308
    ConstraintWeights_1:float=0.3315
    ConstraintWeights_2:float=0.1255
    ConstraintWeights_3:float=0.0847
    ConstraintWeights_4:float=0.0275
   ) -> (Pose) {
    alphaBoolBlend {blendOption=Linear}
    boneToModify {boneName=dyn_leaf_10}
    constraintSetup [
     {targetBone {boneName=dyn_cape03}}
     {targetBone {boneName=dyn_cape02}}
     {targetBone {boneName=dyn_cape01}}
     {targetBone {boneName=backpack_root}}
     {targetBone {boneName=dyn_cape04}}
    ]
    constraintWeights [1,1,1,1,1]
   }
   n27 = "Constraint - dyn_leaf_11 (PPPPPPPPPPPP)"!(
    ComponentPose=@n26
    Alpha=1
    ConstraintWeights_0=0.0534
    ConstraintWeights_1:float=0.0009
    ConstraintWeights_2:float=0.0197
    ConstraintWeights_3:float=0.342
    ConstraintWeights_4:float=0.0001
    ConstraintWeights_5:float=0.1136
    ConstraintWeights_6:float=0
    ConstraintWeights_7:float=0
    ConstraintWeights_8:float=0.4461
    ConstraintWeights_9:float=0.0001
    ConstraintWeights_10:float=0.0233
    ConstraintWeights_11:float=0.0007
   ) -> (Pose) {
    alphaBoolBlend {blendOption=Linear}
    boneToModify {boneName=dyn_leaf_11}
    constraintSetup [
     {targetBone {boneName=dyn_cape03}}
     {targetBone {boneName=dyn_cape03_l}}
     {targetBone {boneName=dyn_cape02_r}}
     {targetBone {boneName=dyn_cape04}}
     {targetBone {boneName=dyn_cape04_l}}
     {targetBone {boneName=dyn_cape01_r}}
     {targetBone {boneName=dyn_cape05_r}}
     {targetBone {boneName=dyn_cape05_l}}
     {targetBone {boneName=dyn_cape01_l}}
     {targetBone {boneName=dyn_cape04_r}}
     {targetBone {boneName=dyn_cape02_l}}
     {targetBone {boneName=dyn_cape03_r}}
    ]
    constraintWeights [1,1,1,1,1,1,1,1,1,1,1,1]
   }
   n28 = "Constraint - dyn_leaf_12 (PPPPPPPPPPPPP)"!(
    ComponentPose=@n27
    Alpha=1
    ConstraintWeights_0=0.0004
    ConstraintWeights_1:float=0.2519
    ConstraintWeights_2:float=0.0392
    ConstraintWeights_3:float=0.1127
    ConstraintWeights_4:float=0.001
    ConstraintWeights_5:float=0.003
    ConstraintWeights_6:float=0.0064
    ConstraintWeights_7:float=0.0007
    ConstraintWeights_8:float=0.0009
    ConstraintWeights_9:float=0.0999
    ConstraintWeights_10:float=0.0035
    ConstraintWeights_11:float=0.468
    ConstraintWeights_12:float=0.0123
   ) -> (Pose) {
    alphaBoolBlend {blendOption=Linear}
    boneToModify {boneName=dyn_leaf_12}
    constraintSetup [
     {targetBone {boneName=dyn_cape03}}
     {targetBone {boneName=dyn_leaf_12}}
     {targetBone {boneName=dyn_cape03_l}}
     {targetBone {boneName=dyn_cape02_r}}
     {targetBone {boneName=dyn_cape04}}
     {targetBone {boneName=dyn_cape04_l}}
     {targetBone {boneName=dyn_cape01_r}}
     {targetBone {boneName=dyn_cape05_r}}
     {targetBone {boneName=dyn_cape05_l}}
     {targetBone {boneName=dyn_cape01_l}}
     {targetBone {boneName=dyn_cape04_r}}
     {targetBone {boneName=dyn_cape02_l}}
     {targetBone {boneName=dyn_cape03_r}}
    ]
    constraintWeights [1,1,1,1,1,1,1,1,1,1,1,1,1]
   }
   n29 = "Constraint - dyn_leaf_13 (PPPPPPPPP)"!(
    ComponentPose=@n28
    Alpha=1
    ConstraintWeights_0=0
    ConstraintWeights_1:float=0.3314
    ConstraintWeights_2:float=0.1085
    ConstraintWeights_3:float=0.1039
    ConstraintWeights_4:float=0.027
    ConstraintWeights_5:float=0.0392
    ConstraintWeights_6:float=0.0687
    ConstraintWeights_7:float=0.2148
    ConstraintWeights_8:float=0.1065
   ) -> (Pose) {
    alphaBoolBlend {blendOption=Linear}
    boneToModify {boneName=dyn_leaf_13}
    constraintSetup [
     {targetBone {boneName=dyn_cape03}}
     {targetBone {boneName=dyn_cape03_l}}
     {targetBone {boneName=dyn_cape02_r}}
     {targetBone {boneName=dyn_cape04_l}}
     {targetBone {boneName=dyn_cape05_r}}
     {targetBone {boneName=dyn_cape05_l}}
     {targetBone {boneName=dyn_cape04_r}}
     {targetBone {boneName=dyn_cape02_l}}
     {targetBone {boneName=dyn_cape03_r}}
    ]
    constraintWeights [1,1,1,1,1,1,1,1,1]
   }
   n30 = "Constraint - dyn_leaf_14 (PPPPPPPPP)"!(
    ComponentPose=@n29
    Alpha=1
    ConstraintWeights_0=0
    ConstraintWeights_1:float=0.3556
    ConstraintWeights_2:float=0.0407
    ConstraintWeights_3:float=0.0032
    ConstraintWeights_4:float=0.0005
    ConstraintWeights_5:float=0.0008
    ConstraintWeights_6:float=0.0016
    ConstraintWeights_7:float=0.5947
    ConstraintWeights_8:float=0.0029
   ) -> (Pose) {
    alphaBoolBlend {blendOption=Linear}
    boneToModify {boneName=dyn_leaf_14}
    constraintSetup [
     {targetBone {boneName=dyn_cape03}}
     {targetBone {boneName=dyn_cape03_l}}
     {targetBone {boneName=dyn_cape02_r}}
     {targetBone {boneName=dyn_cape04_l}}
     {targetBone {boneName=dyn_cape05_r}}
     {targetBone {boneName=dyn_cape05_l}}
     {targetBone {boneName=dyn_cape04_r}}
     {targetBone {boneName=dyn_cape02_l}}
     {targetBone {boneName=dyn_cape03_r}}
    ]
    constraintWeights [1,1,1,1,1,1,1,1,1]
   }
   n31 = "Constraint - dyn_leaf_15 (PPPPPPPPPPP)"!(
    ComponentPose=@n30
    Alpha=1
    ConstraintWeights_0=0
    ConstraintWeights_1:float=0.1861
    ConstraintWeights_2:float=0.075
    ConstraintWeights_3:float=0.2717
    ConstraintWeights_4:float=0.1279
    ConstraintWeights_5:float=0.0193
    ConstraintWeights_6:float=0.0048
    ConstraintWeights_7:float=0.0066
    ConstraintWeights_8:float=0.015
    ConstraintWeights_9:float=0.2612
    ConstraintWeights_10:float=0.0323
   ) -> (Pose) {
    alphaBoolBlend {blendOption=Linear}
    boneToModify {boneName=dyn_leaf_15}
    constraintSetup [
     {targetBone {boneName=dyn_cape03}}
     {targetBone {boneName=dyn_cape03_l}}
     {targetBone {boneName=dyn_cape02_r}}
     {targetBone {boneName=dyn_leaf_15}}
     {targetBone {boneName=dyn_leaf_14}}
     {targetBone {boneName=dyn_cape04_l}}
     {targetBone {boneName=dyn_cape05_r}}
     {targetBone {boneName=dyn_cape05_l}}
     {targetBone {boneName=dyn_cape04_r}}
     {targetBone {boneName=dyn_cape02_l}}
     {targetBone {boneName=dyn_cape03_r}}
    ]
    constraintWeights [1,1,1,1,1,1,1,1,1,1,1]
   }
   n32 = "Constraint - dyn_leaf_16 (PPPPP)"!(
    ComponentPose=@n31
    Alpha=1
    ConstraintWeights_0=0.4093
    ConstraintWeights_1:float=0.0092
    ConstraintWeights_2:float=0.0001
    ConstraintWeights_3:float=0.0001
    ConstraintWeights_4:float=0.5813
   ) -> (Pose) {
    alphaBoolBlend {blendOption=Linear}
    boneToModify {boneName=dyn_leaf_16}
    constraintSetup [
     {targetBone {boneName=dyn_cape03}}
     {targetBone {boneName=dyn_cape02}}
     {targetBone {boneName=dyn_cape01}}
     {targetBone {boneName=backpack_root}}
     {targetBone {boneName=dyn_cape04}}
    ]
    constraintWeights [1,1,1,1,1]
   }
   n33 = "Constraint - dyn_leaf_17 (P)"!(ComponentPose=@n32, Alpha=1, ConstraintWeights_0=1, ConstraintWeights_1:float=0) -> (Pose) {
    alphaBoolBlend {blendOption=Linear}
    boneToModify {boneName=dyn_leaf_17}
    constraintSetup [{targetBone {boneName=backpack_root}}]
    constraintWeights [1,0]
   }
   n34 = "Constraint - dyn_leaf_18 (PPPPP)"!(
    ComponentPose=@n33
    Alpha=1
    ConstraintWeights_0=0.5879
    ConstraintWeights_1:float=0.114
    ConstraintWeights_2:float=0.0112
    ConstraintWeights_3:float=0.0074
    ConstraintWeights_4:float=0.2793
   ) -> (Pose) {
    alphaBoolBlend {blendOption=Linear}
    boneToModify {boneName=dyn_leaf_18}
    constraintSetup [
     {targetBone {boneName=dyn_cape03}}
     {targetBone {boneName=dyn_cape02}}
     {targetBone {boneName=dyn_cape01}}
     {targetBone {boneName=backpack_root}}
     {targetBone {boneName=dyn_cape04}}
    ]
    constraintWeights [1,1,1,1,1]
   }
   n35 = "Constraint - dyn_leaf_19 (PPPPPPP)"!(
    ComponentPose=@n34
    Alpha=1
    ConstraintWeights_0=0.0529
    ConstraintWeights_1:float=0.1355
    ConstraintWeights_2:float=0.072
    ConstraintWeights_3:float=0.0111
    ConstraintWeights_4:float=0.0014
    ConstraintWeights_5:float=0.6683
    ConstraintWeights_6:float=0.0588
   ) -> (Pose) {
    alphaBoolBlend {blendOption=Linear}
    boneToModify {boneName=dyn_leaf_19}
    constraintSetup [
     {targetBone {boneName=dyn_cape03}}
     {targetBone {boneName=dyn_cape02}}
     {targetBone {boneName=dyn_cape01}}
     {targetBone {boneName=dyn_leaf_10}}
     {targetBone {boneName=dyn_cape04}}
     {targetBone {boneName=dyn_leaf_19}}
     {targetBone {boneName=backpack_root}}
    ]
    constraintWeights [1,1,1,1,1,1,1]
   }
   n36 = "Constraint - dyn_leaf_20 (PPPPPPP)"!(
    ComponentPose=@n35
    Alpha=1
    ConstraintWeights_0=0.0021
    ConstraintWeights_1:float=0
    ConstraintWeights_2:float=0.376
    ConstraintWeights_3:float=0.0132
    ConstraintWeights_4:float=0.5948
    ConstraintWeights_5:float=0.0003
    ConstraintWeights_6:float=0.0136
   ) -> (Pose) {
    alphaBoolBlend {blendOption=Linear}
    boneToModify {boneName=dyn_leaf_20}
    constraintSetup [
     {targetBone {boneName=dyn_cape03}}
     {targetBone {boneName=dyn_cape03_l}}
     {targetBone {boneName=dyn_cape02_r}}
     {targetBone {boneName=dyn_cape04}}
     {targetBone {boneName=dyn_cape01_r}}
     {targetBone {boneName=dyn_cape02_l}}
     {targetBone {boneName=dyn_cape03_r}}
    ]
    constraintWeights [1,1,1,1,1,1,1]
   }
   n37 = "Constraint - dyn_leaf_21 (PPPPPPPP)"!(
    ComponentPose=@n36
    Alpha=1
    ConstraintWeights_0=0.0002
    ConstraintWeights_1:float=0.2853
    ConstraintWeights_2:float=0.0001
    ConstraintWeights_3:float=0.0001
    ConstraintWeights_4:float=0.0002
    ConstraintWeights_5:float=0.0036
    ConstraintWeights_6:float=0.5595
    ConstraintWeights_7:float=0.1511
   ) -> (Pose) {
    alphaBoolBlend {blendOption=Linear}
    boneToModify {boneName=dyn_leaf_21}
    constraintSetup [
     {targetBone {boneName=dyn_cape03_l}}
     {targetBone {boneName=dyn_cape02_r}}
     {targetBone {boneName=dyn_cape04_l}}
     {targetBone {boneName=dyn_cape05_r}}
     {targetBone {boneName=dyn_cape04_r}}
     {targetBone {boneName=dyn_cape02_l}}
     {targetBone {boneName=dyn_leaf_21}}
     {targetBone {boneName=dyn_cape03_r}}
    ]
    constraintWeights [1,1,1,1,1,1,1,1]
   }
   n38 = "Constraint - dyn_leaf_22 (PPPPP)"!(
    ComponentPose=@n37
    Alpha=1
    ConstraintWeights_0=0.337
    ConstraintWeights_1:float=0.4876
    ConstraintWeights_2:float=0.081
    ConstraintWeights_3:float=0.0532
    ConstraintWeights_4:float=0.0412
   ) -> (Pose) {
    alphaBoolBlend {blendOption=Linear}
    boneToModify {boneName=dyn_leaf_22}
    constraintSetup [
     {targetBone {boneName=dyn_cape03}}
     {targetBone {boneName=dyn_cape02}}
     {targetBone {boneName=dyn_cape01}}
     {targetBone {boneName=backpack_root}}
     {targetBone {boneName=dyn_cape04}}
    ]
    constraintWeights [1,1,1,1,1]
   }
   n39 = Anim Dynamics - Bone: dyn_leaf_01(ComponentPose=@n38, Alpha=1, ExternalForce="0,0,0") {
    angularDampingOverride=0.9
    bGravityOverrideInSimSpace=true
    bOverrideAngularDamping=true
    bUseGravityOverride=true
    alphaBoolBlend {blendOption=Linear}
    boundBone {boneName=dyn_leaf_01}
    componentAppliedLinearAccClamp=<1000 1000 1000>
    componentLinearAccScale=<0.25 0.25 0.25>
    componentLinearVelScale=<0.25 0.25 0.25>
    gravityOverride=<0.85 0.85 0.85>
    physicsBodyDefinitions [{
      sphereCollisionRadius=0
      boundBone {boneName=dyn_leaf_01}
      boxExtents=<4 3 1>
      constraintSetup {angularLimitsMax=<0 50 40>, angularLimitsMin=<0 -50 -40>, angularTarget=<0 0 0>}
      localJointOffset=<1.25 0 0>
     }]
    relativeSpaceBone {boneName=backpack_root}
   }
   n40 = Anim Dynamics - Bone: dyn_leaf_02(ComponentPose=@n39, Alpha=1, ExternalForce="0,0,0") {
    angularDampingOverride=0.9
    bGravityOverrideInSimSpace=true
    bOverrideAngularDamping=true
    bUseGravityOverride=true
    alphaBoolBlend {blendOption=Linear}
    boundBone {boneName=dyn_leaf_02}
    componentAppliedLinearAccClamp=<1000 1000 1000>
    componentLinearAccScale=<0.25 0.25 0.25>
    componentLinearVelScale=<0.25 0.25 0.25>
    gravityOverride=<0.85 0.85 0.85>
    physicsBodyDefinitions [{
      sphereCollisionRadius=0
      boundBone {boneName=dyn_leaf_02}
      boxExtents=<5 6 2>
      constraintSetup {angularLimitsMax=<0 50 40>, angularLimitsMin=<0 -50 -40>, angularTarget=<0 0 0>}
      localJointOffset=<2 0 0>
     }]
    relativeSpaceBone {boneName=backpack_root}
   }
   n41 = Anim Dynamics - Bone: dyn_leaf_03(ComponentPose=@n40, Alpha=1, ExternalForce="0,0,0") {
    angularDampingOverride=0.9
    bGravityOverrideInSimSpace=true
    bOverrideAngularDamping=true
    bUseGravityOverride=true
    alphaBoolBlend {blendOption=Linear}
    boundBone {boneName=dyn_leaf_03}
    componentAppliedLinearAccClamp=<1000 1000 1000>
    componentLinearAccScale=<0.25 0.25 0.25>
    componentLinearVelScale=<0.25 0.25 0.25>
    gravityOverride=<0.85 0.85 0.85>
    physicsBodyDefinitions [{
      sphereCollisionRadius=0
      boundBone {boneName=dyn_leaf_03}
      boxExtents=<5 4 2>
      constraintSetup {angularLimitsMax=<0 50 40>, angularLimitsMin=<0 -50 -40>, angularTarget=<0 0 0>}
      localJointOffset=<1.5 0 0>
     }]
    relativeSpaceBone {boneName=backpack_root}
   }
   n42 = Anim Dynamics - Bone: dyn_leaf_04(ComponentPose=@n41, Alpha=1, ExternalForce="0,0,0") {
    angularDampingOverride=0.9
    bGravityOverrideInSimSpace=true
    bOverrideAngularDamping=true
    bUseGravityOverride=true
    alphaBoolBlend {blendOption=Linear}
    boundBone {boneName=dyn_leaf_04}
    componentAppliedLinearAccClamp=<1000 1000 1000>
    componentLinearAccScale=<0.25 0.25 0.25>
    componentLinearVelScale=<0.25 0.25 0.25>
    gravityOverride=<0.85 0.85 0.85>
    physicsBodyDefinitions [{
      sphereCollisionRadius=0
      boundBone {boneName=dyn_leaf_04}
      boxExtents=<5 5 2>
      constraintSetup {angularLimitsMax=<0 40 40>, angularLimitsMin=<0 -40 -40>, angularTarget=<0 0 0>}
      localJointOffset=<1 0 0>
     }]
    relativeSpaceBone {boneName=backpack_root}
   }
   n43 = Anim Dynamics - Bone: dyn_leaf_08(ComponentPose=@n42, Alpha=1, ExternalForce="0,0,0") {
    angularDampingOverride=0.9
    bGravityOverrideInSimSpace=true
    bOverrideAngularDamping=true
    bUseGravityOverride=true
    alphaBoolBlend {blendOption=Linear}
    boundBone {boneName=dyn_leaf_08}
    componentAppliedLinearAccClamp=<1000 1000 1000>
    componentLinearAccScale=<0.25 0.25 0.25>
    componentLinearVelScale=<0.25 0.25 0.25>
    gravityOverride=<0.85 0.85 0.85>
    physicsBodyDefinitions [{
      sphereCollisionRadius=0
      boundBone {boneName=dyn_leaf_08}
      boxExtents=<5 6 1>
      constraintSetup {angularLimitsMax=<0 50 40>, angularLimitsMin=<0 -50 -40>, angularTarget=<0 0 0>}
      localJointOffset=<1.5 0 0>
     }]
    relativeSpaceBone {boneName=backpack_root}
   }
   n44 = Anim Dynamics - Bone: dyn_leaf_07(ComponentPose=@n43, Alpha=1, ExternalForce="0,0,0") {
    angularDampingOverride=0.9
    bGravityOverrideInSimSpace=true
    bOverrideAngularDamping=true
    bUseGravityOverride=true
    alphaBoolBlend {blendOption=Linear}
    boundBone {boneName=dyn_leaf_07}
    componentAppliedLinearAccClamp=<1000 1000 1000>
    componentLinearAccScale=<0.25 0.25 0.25>
    componentLinearVelScale=<0.25 0.25 0.25>
    gravityOverride=<0.85 0.85 0.85>
    physicsBodyDefinitions [{
      sphereCollisionRadius=0
      boundBone {boneName=dyn_leaf_07}
      boxExtents=<5 7 1.5>
      constraintSetup {angularLimitsMax=<0 50 40>, angularLimitsMin=<0 -50 -40>, angularTarget=<0 0 0>}
      localJointOffset=<1 0 0>
     }]
    relativeSpaceBone {boneName=backpack_root}
   }
   n45 = Anim Dynamics - Bone: dyn_leaf_06(ComponentPose=@n44, Alpha=1, ExternalForce="0,0,0") {
    angularDampingOverride=0.9
    bGravityOverrideInSimSpace=true
    bOverrideAngularDamping=true
    bUseGravityOverride=true
    alphaBoolBlend {blendOption=Linear}
    boundBone {boneName=dyn_leaf_06}
    componentAppliedLinearAccClamp=<1000 1000 1000>
    componentLinearAccScale=<0.25 0.25 0.25>
    componentLinearVelScale=<0.25 0.25 0.25>
    gravityOverride=<0.85 0.85 0.85>
    physicsBodyDefinitions [{
      sphereCollisionRadius=0
      boundBone {boneName=dyn_leaf_06}
      boxExtents=<6 8 1.5>
      constraintSetup {angularLimitsMax=<0 40 40>, angularLimitsMin=<0 -40 -40>, angularTarget=<0 0 0>}
      localJointOffset=<2 0 0>
     }]
    relativeSpaceBone {boneName=backpack_root}
   }
   n46 = Anim Dynamics - Bone: dyn_leaf_05(ComponentPose=@n45, Alpha=1, ExternalForce="0,0,0") {
    angularDampingOverride=0.9
    bGravityOverrideInSimSpace=true
    bOverrideAngularDamping=true
    bUseGravityOverride=true
    alphaBoolBlend {blendOption=Linear}
    boundBone {boneName=dyn_leaf_05}
    componentAppliedLinearAccClamp=<1000 1000 1000>
    componentLinearAccScale=<0.25 0.25 0.25>
    componentLinearVelScale=<0.25 0.25 0.25>
    gravityOverride=<0.85 0.85 0.85>
    physicsBodyDefinitions [{
      sphereCollisionRadius=0
      boundBone {boneName=dyn_leaf_05}
      boxExtents=<4 6 1.5>
      constraintSetup {angularLimitsMax=<0 40 40>, angularLimitsMin=<0 -40 -40>, angularTarget=<0 0 0>}
      localJointOffset=<1.25 0 0>
     }]
    relativeSpaceBone {boneName=backpack_root}
   }
   n47 = Anim Dynamics - Bone: dyn_leaf_09(ComponentPose=@n46, Alpha=1, ExternalForce="0,0,0") {
    angularDampingOverride=0.9
    bGravityOverrideInSimSpace=true
    bOverrideAngularDamping=true
    bUseGravityOverride=true
    alphaBoolBlend {blendOption=Linear}
    boundBone {boneName=dyn_leaf_09}
    componentAppliedLinearAccClamp=<1000 1000 1000>
    componentLinearAccScale=<0.25 0.25 0.25>
    componentLinearVelScale=<0.25 0.25 0.25>
    gravityOverride=<0.85 0.85 0.85>
    physicsBodyDefinitions [{
      sphereCollisionRadius=0
      boundBone {boneName=dyn_leaf_09}
      boxExtents=<6 8 1>
      constraintSetup {angularLimitsMax=<0 40 40>, angularLimitsMin=<0 -40 -40>, angularTarget=<0 0 0>}
      localJointOffset=<1.5 0 0>
     }]
    relativeSpaceBone {boneName=backpack_root}
   }
   n48 = Anim Dynamics - Bone: dyn_leaf_10(ComponentPose=@n47, Alpha=1, ExternalForce="0,0,0") {
    angularDampingOverride=0.9
    bGravityOverrideInSimSpace=true
    bOverrideAngularDamping=true
    bUseGravityOverride=true
    alphaBoolBlend {blendOption=Linear}
    boundBone {boneName=dyn_leaf_10}
    componentAppliedLinearAccClamp=<1000 1000 1000>
    componentLinearAccScale=<0.25 0.25 0.25>
    componentLinearVelScale=<0.25 0.25 0.25>
    gravityOverride=<0.85 0.85 0.85>
    physicsBodyDefinitions [{
      sphereCollisionRadius=0
      boundBone {boneName=dyn_leaf_10}
      boxExtents=<5 6 1>
      constraintSetup {angularLimitsMax=<0 40 40>, angularLimitsMin=<0 -40 -40>, angularTarget=<0 0 0>}
      localJointOffset=<1.5 0 0>
     }]
    relativeSpaceBone {boneName=backpack_root}
   }
   n49 = Anim Dynamics - Bone: dyn_leaf_11(ComponentPose=@n48, Alpha=1, ExternalForce="0,0,0") {
    angularDampingOverride=0.9
    bGravityOverrideInSimSpace=true
    bOverrideAngularDamping=true
    bUseGravityOverride=true
    alphaBoolBlend {blendOption=Linear}
    boundBone {boneName=dyn_leaf_11}
    componentAppliedLinearAccClamp=<1000 1000 1000>
    componentLinearAccScale=<0.25 0.25 0.25>
    componentLinearVelScale=<0.25 0.25 0.25>
    gravityOverride=<0.85 0.85 0.85>
    physicsBodyDefinitions [{
      sphereCollisionRadius=0
      boundBone {boneName=dyn_leaf_11}
      boxExtents=<5 6 1>
      constraintSetup {angularLimitsMax=<0 50 40>, angularLimitsMin=<0 -50 -40>, angularTarget=<0 0 0>}
      localJointOffset=<1.5 0 0>
     }]
    relativeSpaceBone {boneName=backpack_root}
   }
   n50 = Anim Dynamics - Bone: dyn_leaf_12(ComponentPose=@n49, Alpha=1, ExternalForce="0,0,0") {
    angularDampingOverride=0.9
    bGravityOverrideInSimSpace=true
    bOverrideAngularDamping=true
    bUseGravityOverride=true
    alphaBoolBlend {blendOption=Linear}
    boundBone {boneName=dyn_leaf_12}
    componentAppliedLinearAccClamp=<1000 1000 1000>
    componentLinearAccScale=<0.25 0.25 0.25>
    componentLinearVelScale=<0.25 0.25 0.25>
    gravityOverride=<0.85 0.85 0.85>
    physicsBodyDefinitions [{
      sphereCollisionRadius=0
      boundBone {boneName=dyn_leaf_12}
      boxExtents=<5 6 1>
      constraintSetup {angularLimitsMax=<0 40 40>, angularLimitsMin=<0 -40 -40>, angularTarget=<0 0 0>}
      localJointOffset=<1.5 0 0>
     }]
    relativeSpaceBone {boneName=backpack_root}
   }
   n51 = Anim Dynamics - Bone: dyn_leaf_13(ComponentPose=@n50, Alpha=1, ExternalForce="0,0,0") {
    angularDampingOverride=0.9
    bGravityOverrideInSimSpace=true
    bOverrideAngularDamping=true
    bUseGravityOverride=true
    alphaBoolBlend {blendOption=Linear}
    boundBone {boneName=dyn_leaf_13}
    componentAppliedLinearAccClamp=<1000 1000 1000>
    componentLinearAccScale=<0.25 0.25 0.25>
    componentLinearVelScale=<0.25 0.25 0.25>
    gravityOverride=<0.85 0.85 0.85>
    physicsBodyDefinitions [{
      sphereCollisionRadius=0
      boundBone {boneName=dyn_leaf_13}
      boxExtents=<8 6 1>
      constraintSetup {angularLimitsMax=<0 50 40>, angularLimitsMin=<0 -60 -40>, angularTarget=<0 0 0>}
      localJointOffset=<2 0 0>
     }]
    relativeSpaceBone {boneName=backpack_root}
   }
   n52 = Anim Dynamics - Bone: dyn_leaf_14(ComponentPose=@n51, Alpha=1, ExternalForce="0,0,0") {
    angularDampingOverride=0.9
    bGravityOverrideInSimSpace=true
    bOverrideAngularDamping=true
    bUseGravityOverride=true
    alphaBoolBlend {blendOption=Linear}
    boundBone {boneName=dyn_leaf_14}
    componentAppliedLinearAccClamp=<1000 1000 1000>
    componentLinearAccScale=<0.25 0.25 0.25>
    componentLinearVelScale=<0.25 0.25 0.25>
    gravityOverride=<0.85 0.85 0.85>
    physicsBodyDefinitions [{
      sphereCollisionRadius=0
      boundBone {boneName=dyn_leaf_14}
      boxExtents=<8 6 1>
      constraintSetup {angularLimitsMax=<0 50 40>, angularLimitsMin=<0 -50 -40>, angularTarget=<0 0 0>}
      localJointOffset=<2 0 0>
     }]
    relativeSpaceBone {boneName=backpack_root}
   }
   n53 = Anim Dynamics - Bone: dyn_leaf_15(ComponentPose=@n52, Alpha=1, ExternalForce="0,0,0") {
    angularDampingOverride=0.9
    bGravityOverrideInSimSpace=true
    bOverrideAngularDamping=true
    bUseGravityOverride=true
    alphaBoolBlend {blendOption=Linear}
    boundBone {boneName=dyn_leaf_15}
    componentAppliedLinearAccClamp=<1000 1000 1000>
    componentLinearAccScale=<0.25 0.25 0.25>
    componentLinearVelScale=<0.25 0.25 0.25>
    gravityOverride=<0.85 0.85 0.85>
    physicsBodyDefinitions [{
      sphereCollisionRadius=0
      boundBone {boneName=dyn_leaf_15}
      boxExtents=<5 6 1>
      constraintSetup {angularLimitsMax=<0 50 40>, angularLimitsMin=<0 -50 -40>, angularTarget=<0 0 0>}
      localJointOffset=<1.5 0 0>
     }]
    relativeSpaceBone {boneName=backpack_root}
   }
   n54 = Anim Dynamics - Bone: dyn_leaf_16(ComponentPose=@n53, Alpha=1, ExternalForce="0,0,0") {
    angularDampingOverride=0.9
    bGravityOverrideInSimSpace=true
    bOverrideAngularDamping=true
    bUseGravityOverride=true
    alphaBoolBlend {blendOption=Linear}
    boundBone {boneName=dyn_leaf_16}
    componentAppliedLinearAccClamp=<1000 1000 1000>
    componentLinearAccScale=<0.25 0.25 0.25>
    componentLinearVelScale=<0.25 0.25 0.25>
    gravityOverride=<0.85 0.85 0.85>
    physicsBodyDefinitions [{
      sphereCollisionRadius=0
      boundBone {boneName=dyn_leaf_16}
      boxExtents=<6 8 1>
      constraintSetup {angularLimitsMax=<0 20 40>, angularLimitsMin=<0 -30 -40>, angularTarget=<0 0 0>}
      localJointOffset=<1.5 0 0>
     }]
    relativeSpaceBone {boneName=backpack_root}
   }
   n55 = Anim Dynamics - Bone: dyn_leaf_17(ComponentPose=@n54, Alpha=1, ExternalForce="0,0,0") {
    angularDampingOverride=0.9
    bGravityOverrideInSimSpace=true
    bOverrideAngularDamping=true
    bUseGravityOverride=true
    alphaBoolBlend {blendOption=Linear}
    boundBone {boneName=dyn_leaf_17}
    componentAppliedLinearAccClamp=<1000 1000 1000>
    componentLinearAccScale=<0.25 0.25 0.25>
    componentLinearVelScale=<0.25 0.25 0.25>
    gravityOverride=<0.85 0.85 0.85>
    physicsBodyDefinitions [{
      sphereCollisionRadius=0
      boundBone {boneName=dyn_leaf_17}
      boxExtents=<8 6 1>
      constraintSetup {angularLimitsMax=<0 50 40>, angularLimitsMin=<0 -50 -40>, angularTarget=<0 0 0>}
      localJointOffset=<3 0 0>
     }]
    relativeSpaceBone {boneName=backpack_root}
   }
   n56 = Anim Dynamics - Bone: dyn_leaf_18(ComponentPose=@n55, Alpha=1, ExternalForce="0,0,0") {
    angularDampingOverride=0.9
    bGravityOverrideInSimSpace=true
    bOverrideAngularDamping=true
    bUseGravityOverride=true
    alphaBoolBlend {blendOption=Linear}
    boundBone {boneName=dyn_leaf_18}
    componentAppliedLinearAccClamp=<1000 1000 1000>
    componentLinearAccScale=<0.25 0.25 0.25>
    componentLinearVelScale=<0.25 0.25 0.25>
    gravityOverride=<0.85 0.85 0.85>
    physicsBodyDefinitions [{
      sphereCollisionRadius=0
      boundBone {boneName=dyn_leaf_18}
      boxExtents=<8 6 1>
      constraintSetup {angularLimitsMax=<0 40 40>, angularLimitsMin=<0 -50 -40>, angularTarget=<0 0 0>}
      localJointOffset=<2.5 0 0>
     }]
    relativeSpaceBone {boneName=backpack_root}
   }
   n57 = Anim Dynamics - Bone: dyn_leaf_19(ComponentPose=@n56, Alpha=1, ExternalForce="0,0,0") {
    angularDampingOverride=0.9
    bGravityOverrideInSimSpace=true
    bOverrideAngularDamping=true
    bUseGravityOverride=true
    alphaBoolBlend {blendOption=Linear}
    boundBone {boneName=dyn_leaf_19}
    componentAppliedLinearAccClamp=<1000 1000 1000>
    componentLinearAccScale=<0.25 0.25 0.25>
    componentLinearVelScale=<0.25 0.25 0.25>
    gravityOverride=<0.85 0.85 0.85>
    physicsBodyDefinitions [{
      sphereCollisionRadius=0
      boundBone {boneName=dyn_leaf_19}
      boxExtents=<8 6 1>
      constraintSetup {angularLimitsMax=<0 40 40>, angularLimitsMin=<0 -50 -40>, angularTarget=<0 0 0>}
      localJointOffset=<2.5 0 0>
     }]
    relativeSpaceBone {boneName=backpack_root}
   }
   n58 = Anim Dynamics - Bone: dyn_leaf_20(ComponentPose=@n57, Alpha=1, ExternalForce="0,0,0") {
    angularDampingOverride=0.9
    bGravityOverrideInSimSpace=true
    bOverrideAngularDamping=true
    bUseGravityOverride=true
    alphaBoolBlend {blendOption=Linear}
    boundBone {boneName=dyn_leaf_20}
    componentAppliedLinearAccClamp=<1000 1000 1000>
    componentLinearAccScale=<0.25 0.25 0.25>
    componentLinearVelScale=<0.25 0.25 0.25>
    gravityOverride=<0.85 0.85 0.85>
    physicsBodyDefinitions [{
      sphereCollisionRadius=0
      boundBone {boneName=dyn_leaf_20}
      boxExtents=<8 6 1>
      constraintSetup {angularLimitsMax=<0 50 40>, angularLimitsMin=<0 -50 -40>, angularTarget=<0 0 0>}
      localJointOffset=<2.5 0 0>
     }]
    relativeSpaceBone {boneName=backpack_root}
   }
   n59 = Anim Dynamics - Bone: dyn_leaf_21(ComponentPose=@n58, Alpha=1, ExternalForce="0,0,0") {
    angularDampingOverride=0.9
    bGravityOverrideInSimSpace=true
    bOverrideAngularDamping=true
    bUseGravityOverride=true
    alphaBoolBlend {blendOption=Linear}
    boundBone {boneName=dyn_leaf_21}
    componentAppliedLinearAccClamp=<1000 1000 1000>
    componentLinearAccScale=<0.25 0.25 0.25>
    componentLinearVelScale=<0.25 0.25 0.25>
    gravityOverride=<0.85 0.85 0.85>
    physicsBodyDefinitions [{
      sphereCollisionRadius=0
      boundBone {boneName=dyn_leaf_21}
      boxExtents=<6 8 1>
      constraintSetup {angularLimitsMax=<0 50 40>, angularLimitsMin=<0 -60 -40>, angularTarget=<0 0 0>}
      localJointOffset=<2.5 0 0>
     }]
    relativeSpaceBone {boneName=backpack_root}
   }
   n60 = Anim Dynamics - Bone: dyn_leaf_22(ComponentPose=@n59, Alpha=1, ExternalForce="0,0,0") {
    angularDampingOverride=0.9
    bGravityOverrideInSimSpace=true
    bOverrideAngularDamping=true
    bUseGravityOverride=true
    alphaBoolBlend {blendOption=Linear}
    boundBone {boneName=dyn_leaf_22}
    componentAppliedLinearAccClamp=<1000 1000 1000>
    componentLinearAccScale=<0.25 0.25 0.25>
    componentLinearVelScale=<0.25 0.25 0.25>
    gravityOverride=<0.85 0.85 0.85>
    physicsBodyDefinitions [{
      sphereCollisionRadius=0
      boundBone {boneName=dyn_leaf_22}
      boxExtents=<8 6 1>
      constraintSetup {angularLimitsMax=<0 50 40>, angularLimitsMin=<0 -50 -40>, angularTarget=<0 0 0>}
      localJointOffset=<2.5 0 0>
     }]
    relativeSpaceBone {boneName=backpack_root}
   }
   Output Pose(Result=Component To Local(ComponentPose=@n60)) {layerGroup=None}
   Save cached pose 'Cape_Cache'(Pose=Copy Pose From Mesh() {bUseAttachedParent=true}) {cacheName=Cape_Cache}
   n61 = Get bIsDBNO()
   SelectFloat(A=SelectFloat(A=0.5, B=Get BlendOutForksAlpha() -> (BlendOutForksAlpha:float), bPickA=@n61), B=1, bPickA=@n5)
   SelectVector(A=Get FakeWindVelocityDBNO() -> (FakeWindVelocityDBNO:Vector), B=Get FakeWindVelocity() -> (FakeWindVelocity:Vector), bPickA=@n61)
  ]}
 EventGraph {nodes [
   Event BlueprintUpdateAnimation()
   > n1 = Sequence!(execute) -> (then_0, then_1, then_2:Exec)
   > n2 = Branch(Condition=Get bIsInFrontEnd())
   > Set BwdMovement(BwdMovement:double=0) -> (Output_Get:double)
   n3 = PlayerMovement(execute=@n2.else, AnimInstance=Self-Reference(), XDotProductCoefficient=1, YDotProductCoefficient=1, ZDotProductCoefficient=1)
   > n4 = Branch(Condition=Get bIsCrouching())
   n5 = MapRangeClamped(Value=@n3.YMovement, InRangeA=0, InRangeB=-1)
   n6 = float * float(A="Absolute (Float)"(A=@n3), B=0.5)
   > Set BwdMovement(BwdMovement=MapRangeClamped(
     Value=float + float(A=@n6, B=@n5) {is=title K2Node_CommutativeAssociativeBinaryOperator}
     InRangeA=0
     InRangeB=1.1
     OutRangeA=0.8
     OutRangeB=0
    ))
   > Set FwdMovement(FwdMovement:double=MapRangeClamped(Value=@n3.YMovement, InRangeA=1, InRangeB=0)) -> (Output_Get:double)
   n7 = float + float(A=float * float(A="Absolute (Float)"(A=@n3), B=0.5), B=MapRangeClamped(Value=@n3.YMovement, InRangeA=0, InRangeB=-1))
   Set BwdMovement(execute=@n4.else, BwdMovement=MapRangeClamped(Value=@n7, InRangeA=0, InRangeB=1.1, OutRangeA=0.8, OutRangeB=0))
   > Set FwdMovement(FwdMovement=MapRangeClamped(Value=@n7, InRangeA=0, InRangeB=1.1))
   n8 = Branch(execute=@n1.then_1, Condition=Get bIsSkyDiving())
   > n9 = Branch(Condition="float > float"(A=Get PawnVelocityZ(), B=50) {is=title K2Node_CallFunction})
   > Set BwdMovement(BwdMovement=float * float(A=Get BwdMovement(), B=0))
   Set BwdMovement(execute=[@n9.else, @n8.else], BwdMovement=float * float(A=Get BwdMovement(), B=1))
   n10 = Get bIsDBNO()
   SelectFloat(A=SelectFloat(A=0.5, B=Get BlendOutForksAlpha(), bPickA=@n10), B=1, bPickA=false)
   SelectVector(A=Get FakeWindVelocityDBNO(), B=Get FakeWindVelocity(), bPickA=@n10)
   n11 = "Absolute (Float)"(A=@n3.YMovement)
   MapRangeClamped(
    Value=float + float(A=float * float(A=@n11, B=0.2), B=MapRangeClamped(Value=@n3, InRangeA=0.2, InRangeB=-1, OutRangeB=0.3))
    InRangeA=0
    InRangeB=1.2
    OutRangeB=0.7
   )
   n12 = "Absolute (Float)"(A=@n3.YMovement)
   MapRangeClamped(
    Value=float + float(A=float * float(A=@n12, B=0.2), B=MapRangeClamped(Value=@n3, InRangeA=-0.2, InRangeB=1))
    InRangeA=0
    InRangeB=1.2
    OutRangeB=0.7
   )
  ]}
]
variables [
 BwdMovement {type=number}
 FwdMovement {type=number}
]
